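# Sound cue for the Tapestry emote's foley audio, mixing a start one-shot with delayed looping layers.
# A crossfade on EmoteMoveSpeed blends the idle loop into the movement loop, while a shimmer loop runs alongside.
# All three branches override the cue's foley sound class to Fort_Music_Emote.
# The cue uses emote attenuation and concurrency rules and has a send to the music audio-analysis submix.
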SoundCue /BRCosmetics/Sounds/Emotes/Tapestry/EmoteFoley_Tapestry_Cue
root=Mixer_0
nodes [
 {id=Delay_0, children [WavePlayer_1], properties {delayMax=0.8, delayMin=0.8}}
 {id=Delay_1, children [WavePlayer_4], properties {delayMax=0.8, delayMin=0.8}}
 {id=Delay_2, children [WavePlayer_2], properties {delayMax=0.7, delayMin=0.7}}
 {id=Mixer_0, children [SoundClass_1,SoundClass_0,SoundClass_2], properties {inputVolume [1,1,1]}}
 {
  id=ParamCrossFade_0
  children [Delay_0,Delay_1]
  properties {
   paramName=EmoteMoveSpeed
   crossFadeInput [
    {fadeInDistanceEnd=0, fadeInDistanceStart=0, fadeOutDistanceEnd=1, fadeOutDistanceStart=0, volume=1}
    {fadeInDistanceEnd=1, fadeInDistanceStart=0, fadeOutDistanceEnd=100, fadeOutDistanceStart=100, volume=1}
   ]
  }
 }
 {id=SoundClass_0, children [ParamCrossFade_0], properties {soundClassOverride=&Fort_Music_Emote}}
 {id=SoundClass_1, children [WavePlayer_0], properties {soundClassOverride=&Fort_Music_Emote}}
 {id=SoundClass_2, children [Delay_2], properties {soundClassOverride=&Fort_Music_Emote}}
 {id=WavePlayer_0, properties {soundWaveAssetPtr=&/Emote_Tapestry_Start}}
 {id=WavePlayer_1, properties {bLooping=true, soundWaveAssetPtr=&/Emote_Tapestry_Idle_Loop}}
 {id=WavePlayer_2, properties {bLooping=true, soundWaveAssetPtr=&/Emote_Tapestry_Shimmer_Loop}}
 {id=WavePlayer_4, properties {bLooping=true, soundWaveAssetPtr=&/Emote_Tapestry_Loop}}
]
properties {
 attenuationSettings=&Fort_Emotes_Athena_Att
 audioPropertiesSheet=/BRCosmetics/Sounds/Settings/PropertySheets/Emotes/APS_Cosmetic_Emote_Foley
 maxDistance=1600
 soundClassObject=/Game/Sounds/FortSoundClassesAndModes/Fort_Foley_Emote
 concurrencySet [/Game/Sounds/Concurrency_Rules/Athena_EmoteFoley_Conc]
 soundSubmixSends [{minSendLevel=0, soundSubmix=&Fortnite_AudioAnalysis_Music}]
}
